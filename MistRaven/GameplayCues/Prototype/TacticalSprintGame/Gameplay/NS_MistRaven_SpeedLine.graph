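# Niagara speed-line effect for the MistRaven TacticalSprintGame prototype cue, emitting faint, thin, short-lived white sprite streaks around a skeletal mesh.
# Spawn rate uses the length of the normalized owner velocity, so it acts as a moving/stationary switch rather than scaling with speed.
# Sprites spawn on the mesh with spherical offset, fly along negative Y, and change size over life.
# An alpha curve fades them, scaled by User.energy only when User.bIsLocalClient is true.
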NiagaraSystem /MistRaven/GameplayCues/Prototype/TacticalSprintGame/Gameplay/NS_MistRaven_SpeedLine
dependencies {
 usedDataInterfaces [
  &NiagaraDataInterfaceCurve
  &NiagaraDataInterfaceVector2DCurve
  &NiagaraDataInterfaceSkeletalMesh
 ]
 usedDynamicInputs []
 usedModules [
  &self:SystemSpawnScript.SystemScriptSource.SystemScriptGraph.NiagaraNodeAssignment_0.SetVariables_#1
  &SystemState
 ]
}
emitters {SpeedLine_Particles {
  emitterUpdateScript [
   &EmitterState()
   &SpawnRate(
    SpawnRate=&Multiply_Float(
     A=30
     B=&VectorLength(Vector=&NormalizeVector(Vector To Normalize=$Engine.Owner.Velocity, Safe Normalize=true, Safe Fallback Threshold=0, Safe Fallback Vector=<0 0 0.001>))
    )
    Use Spawn Probability=false
   )
  ]
  particleSpawnScript [
   &InitializeParticle(
    Color=&RandomRangeLinearColor(Minimum=<rgba 1 1 1 0.02>, Maximum=<rgba 1 1 1 0.05>)
    Lifetime=&RandomRangeFloat(Minimum=0.2, Maximum=0.3)
    Sprite Size Max=<1 70>
    Sprite Size Min=<1 42>
    Sprite Size Mode=::NewEnumerator2
    UsePositionOffset=false
    Write Material Random=true
   )
   &SkeletalMeshLocation(MaskSpawnGroup=false, OverrideRandomSeed=false, Skeletal Mesh=$System.Skeletal Mesh)
   &ShapeLocation(MaskSpawnGroup=false, Override Seed=false, Sphere Radius=17, UseOwnerScale=true)
   &AddVelocity(Velocity=&RandomRangeVector(Minimum=<0 -800 0>, Maximum=<0 -1100 0>))
  ]
  particleUpdateScript [
   &ParticleState()
   &SolveForcesAndVelocity(Clamp Velocity=false, Limit Acceleration=false)
   &ScaleColor(
    Scale Alpha=&FloatFromCurve(
     FloatCurve=DataInterface() {propertyValues {Curve {keys [
         {interpMode=RCIM_Cubic, leaveTangent=12.945, tangentMode=RCTM_Break, time=0, value=0}
         {interpMode=RCIM_Cubic, tangentMode=RCTM_Break, time=0.2, value=1}
         {arriveTangent=-1.8151, interpMode=RCIM_Cubic, tangentMode=RCTM_Break, time=1, value=0}
        ]}}}
     Scale Curve=&MakeCustomFloatFromBool(Bool=$User.bIsLocalClient, True Float=$User.energy, False Float=1)
    )
    Scale RGB=<0.8 0.8 0.8>
    ScaleA=true
    ScaleRGB=true
   )
   &ScaleSpriteSize(
    Non-Uniform Curve Sprite Scale=DataInterface() {propertyValues {
      XCurve {keys [{tangentMode=RCTM_User, time=0.55, value=1}, {tangentMode=RCTM_User, time=1, value=0}]}
      YCurve {keys [
        {interpMode=RCIM_Cubic, leaveTangent=0.9226, tangentMode=RCTM_Break, time=0, value=0.55}
        {arriveTangent=0.0265, interpMode=RCIM_Cubic, tangentMode=RCTM_Break, time=1, value=1}
       ]}
     }}
    Scale Sprite Size Mode=::NewEnumerator3
   )
  ]
  rendererClasses [&NiagaraSpriteRendererProperties]
  renderers [{rendererClass=&NiagaraSpriteRendererProperties}]
 }}
summary {
 systemName=NS_MistRaven_SpeedLine
 emitters [{emitterName=SpeedLine_Particles, rendererClasses [&NiagaraSpriteRendererProperties]}]
 userVariables [
  User.bIsLocalClient {type=NiagaraBool}
  User.energy {type=NiagaraFloat, defaultValue=0, description=""}
 ]
}
